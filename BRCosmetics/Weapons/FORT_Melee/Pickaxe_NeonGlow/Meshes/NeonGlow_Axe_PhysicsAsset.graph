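# Physics asset for the NeonGlow pickaxe melee weapon mesh, providing its simple collision and physics shape.
# It holds a single body on the Melee_TwoHanded_Handle bone, built from one capsule offset and rotated from the bone origin.
# It has no constraints, so the weapon acts as one rigid body with no articulated parts.
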PhysicsAsset /BRCosmetics/Weapons/FORT_Melee/Pickaxe_NeonGlow/Meshes/NeonGlow_Axe_PhysicsAsset
bodies {Melee_TwoHanded_Handle [{shapeType=Capsule, length=24.0372, radius=4.6628, center=<-0.1377 -0.5384 12.0712>, rotation=<rot -0.0286 -90 0.089>}]}
constraints []
