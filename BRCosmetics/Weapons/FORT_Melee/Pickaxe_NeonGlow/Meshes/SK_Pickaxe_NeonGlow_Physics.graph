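# Physics asset for the SK_Pickaxe_NeonGlow skeletal mesh, a cosmetic two-handed melee pickaxe.
# It holds a single body on the Melee_TwoHanded_Handle bone, made of one box shape offset from the bone origin.
# It defines no constraints, so the weapon collides as one rigid piece.
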PhysicsAsset /BRCosmetics/Weapons/FORT_Melee/Pickaxe_NeonGlow/Meshes/SK_Pickaxe_NeonGlow_Physics
bodies {Melee_TwoHanded_Handle [{shapeType=Box, extentX=13.8109, extentY=82.5967, extentZ=129.9931, center=<0.8145 -1.2814 20.385>}]}
constraints []
